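# Translucent UI master material that draws a two-color radial gradient for frontend widgets.
# Normalized UI UVs, remapped around zero, feed an exponential radial gradient with adjustable center, radius and density, which blends RGB_A and RGB_B by both additive and lerp nodes.
# RadiusPerfectCircle is meant to keep the gradient circular on rectangular brushes, but its scale divides pixel width by itself, so it applies no aspect correction.
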Material /Game/Athena/UI/Frontend/Art/M_UI_RadialGradientMaster
graph {
 properties {
  bUsedWithStaticMesh=false
  blendMode=BLEND_Translucent
  materialDomain=MD_UI
  pixelDepthOffsetMode=PDOM_Legacy
  usedShadingModels=Default Lit
 }
 nodes [
  n1 = VectorParameter() {parameterName=RGB_A, defaultValue=<rgba 0.0194 0.0666 0.4179 1>}
  n2 = MaterialFunctionCall() {materialFunction=&GetUserInterfaceUV}
  n3 = Constant() {r=1}
  n4 = MaterialFunctionCall(Float2=@n2.Pixel Size) {materialFunction=&BreakOutFloat2Components}
  n5 = Divide(A=@n4.R, B=@n4.R)
  n6 = ScalarParameter() {
   desc="Set to 1 if you want the Radial Gradient to always be circular even when the brush size is rectangular (Fit to X)"
   parameterName=RadiusPerfectCircle
  }
  n7 = LinearInterpolate(
   A=@n2.Normalized UV
   B=MaterialFunctionCall(UVs=@n2.Normalized UV, Texture Scale=AppendVector(A=@n3, B=@n5)) {materialFunction=&ScaleUVsByCenter}.UVs
   Alpha=@n6
  )
  n8 = MaterialFunctionCall(
   UVs=Multiply(A=Subtract(A=@n7) {constB=0.5}) {constB=2}
   CenterPosition=MaterialFunctionCall(X=ScalarParameter() {parameterName=Center_X}, Y=ScalarParameter() {parameterName=Center_Y}) {materialFunction=&MakeFloat2}.Result
   Radius=ScalarParameter() {defaultValue=1, parameterName=Radius}
   Density=ScalarParameter() {defaultValue=2.33, parameterName=Density}
  ) {materialFunction=&RadialGradientExponential}
  n9 = VectorParameter() {parameterName=RGB_B, defaultValue=<rgba 0 0.4342 0.4233 1>}
  Add(A=@n1.RGB, B=Multiply(A=@n8.RadialGradientExponential, B=@n9.RGB))
  LinearInterpolate(A=@n1.RGB, B=@n9.RGB, Alpha=@n8.RadialGradientExponential)
  ScalarParameter() {defaultValue=1, parameterName=OverallAlpha}
 ]
}
