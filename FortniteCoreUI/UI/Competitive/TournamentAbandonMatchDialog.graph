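# Tournament dialog offering to abandon a match without it counting when the full team failed to join, or continue and be scored.
# It is a layout-only subclass of FortTournamentAbandonMatchDialog; its event graph holds only empty stub events.
# Over a scaled radial gradient, a centered column shows header, explanation, the inherited Text_LeaveMatchTime date-time text and two inherited buttons.
# Abandon (green) and Continue (red) are bound to bottom and right face-button input actions.
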WidgetBlueprint /FortniteCoreUI/UI/Competitive/TournamentAbandonMatchDialog
graphs [EventGraph {nodes [
   Event PreConstruct()
   Event Construct()
   Event Tick()
  ]}]
variables []
widgets {
 info {
  inheritedWidgetCount=3
  parentClass=/Script/FortniteUI.FortTournamentAbandonMatchDialog
  rootWidgetClass=&SafeZone
  widgetCount=14
 }
 widgets [
  SafeZone_30 {widgetClassPath=&SafeZone, widget=SafeZone_30}
  Overlay_38 {widgetClassPath=&Overlay, parent=SafeZone_30, widget=Overlay_38, slot {class=SafeZoneSlot}}
  ScaleBox_1 {
   widgetClassPath=&ScaleBox
   parent=Overlay_38
   widget=ScaleBox_1
   properties {stretch=ScaleToFill}
   slot {class=OverlaySlot, properties {horizontalAlignment=HAlign_Fill, verticalAlignment=VAlign_Fill}}
  }
  Overlay_3 {
   widgetClassPath=&Overlay
   parent=ScaleBox_1
   widget=Overlay_3
   slot {class=ScaleBoxSlot, properties {horizontalAlignment=HAlign_Fill}}
  }
  BackgroundGradient {
   widgetClassPath=&Image
   parent=Overlay_3
   widget=BackgroundGradient
   properties {
    visibility=SelfHitTestInvisible
    brush {
     bIsDynamicallyLoaded=false
     drawAs=Image
     imageType=NoImage
     mirroring=NoMirror
     resourceName=None
     resourceObject=/Game/Athena/UI/Frontend/Art/M_UI_RadialGradientMaster
     tiling=NoTile
     imageSize=<2160 2160>
     margin=<ltrb 0 0 0 0>
     outlineSettings {
      bUseBrushTransparency=false
      roundingType=HalfHeightRadius
      width=0
      color {colorUseRule=UseColor_Specified, specifiedColor=<rgba 0 0 0 0>}
      cornerRadii=<0 0 0 0>
     }
     tintColor {colorUseRule=UseColor_Specified, specifiedColor=<rgba 1 1 1 1>}
     uVRegion {bIsValid=false, max=<0 0>, min=<0 0>}
    }
   }
   slot {class=OverlaySlot}
  }
  SizeBox_0 {
   widgetClassPath=&SizeBox
   parent=Overlay_38
   widget=SizeBox_0
   properties {bOverride_WidthOverride=true, widthOverride=1000}
   slot {class=OverlaySlot, properties {horizontalAlignment=HAlign_Center, verticalAlignment=VAlign_Center}}
  }
  VerticalBox_52 {widgetClassPath=&VerticalBox, parent=SizeBox_0, widget=VerticalBox_52, slot {class=SizeBoxSlot}}
  CommonTextBlock_114 {
   widgetClassPath=&CommonTextBlock
   parent=VerticalBox_52
   widget=CommonTextBlock_114
   properties {
    style=/FortUILibrary/Styles/Typography/Foundation/TextStyle-Header-L.TextStyle-Header-L_C
    text="Incomplete Team - Abandon Match?!"
    font {
     bForceMonospaced=false
     bMaterialIsStencil=false
     fontMaterial=null
     fontObject=&Burbank
     letterSpacing=0
     monospacedWidth=1
     size=48
     skewAmount=0
     typefaceFontName=Burbank-Black
     outlineSettings {
      bApplyOutlineToDropShadows=false
      bMiteredCorners=false
      bSeparateFillAlpha=false
      outlineMaterial=null
      outlineSize=0
      outlineColor=<rgba 0 0 0 1>
     }
    }
    shadowOffset=<0 0>
   }
   slot {class=VerticalBoxSlot, properties {padding=<ltrb 0 0 0 25>}}
  }
  CommonTextBlock_52 {
   widgetClassPath=&CommonTextBlock
   parent=VerticalBox_52
   widget=CommonTextBlock_52
   properties {
    autoWrapText=true
    style=/FortUILibrary/Styles/Typography/Foundation/TextStyle-Base-M-B.TextStyle-Base-M-B_C
    text="Your entire team did not make it into the match.\r\n\r\nYou may abandon match within the next few seconds and this match will not count for your team.\r\n\r\nIf you continue, the match will be scored."
    font {
     bForceMonospaced=false
     bMaterialIsStencil=false
     fontMaterial=null
     fontObject=&NotoSans
     letterSpacing=0
     monospacedWidth=1
     size=24
     skewAmount=0
     typefaceFontName=NotoSans-Bold
     outlineSettings {
      bApplyOutlineToDropShadows=false
      bMiteredCorners=false
      bSeparateFillAlpha=false
      outlineMaterial=null
      outlineSize=0
      outlineColor=<rgba 0 0 0 1>
     }
    }
    shadowOffset=<0 0>
   }
   slot {class=VerticalBoxSlot, properties {padding=<ltrb 0 0 0 25>}}
  }
  Text_LeaveMatchTime {
   widgetClassPath=/Script/CommonUI.CommonDateTimeTextBlock
   bInherited=true
   parent=VerticalBox_52
   widget=Text_LeaveMatchTime
   properties {
    justification=Center
    style=/FortUILibrary/Styles/Typography/Foundation/TextStyle-Header-M.TextStyle-Header-M_C
    text="31 Dec 1, 23:58:45"
    font {
     bForceMonospaced=false
     bMaterialIsStencil=false
     fontMaterial=null
     fontObject=&Burbank
     letterSpacing=0
     monospacedWidth=1
     size=32
     skewAmount=0
     typefaceFontName=Burbank-Black
     outlineSettings {
      bApplyOutlineToDropShadows=false
      bMiteredCorners=false
      bSeparateFillAlpha=false
      outlineMaterial=null
      outlineSize=0
      outlineColor=<rgba 0 0 0 1>
     }
    }
    shadowOffset=<0 0>
   }
   slot {class=VerticalBoxSlot, properties {padding=<ltrb 0 0 0 25>}}
  }
  HorizontalBox_47 {widgetClassPath=&HorizontalBox, parent=VerticalBox_52, widget=HorizontalBox_47, slot {class=VerticalBoxSlot}}
  Button_LeaveMatch {
   widgetClassPath=&IconTextButton
   bInherited=true
   bIsVariable=true
   parent=HorizontalBox_47
   widget=Button_LeaveMatch
   properties {
    buttonText=Abandon Match
    onButtonBaseClicked=(&self:WidgetTree.Button_LeaveMatch.HandleOnButtonClicked)
    onButtonBaseDoubleClicked=(&self:WidgetTree.Button_LeaveMatch.HandleOnButtonDoubleClicked)
    onButtonBaseHovered=(&self:WidgetTree.Button_LeaveMatch.HandleOnButtonHovered)
    onButtonBaseUnhovered=(&self:WidgetTree.Button_LeaveMatch.HandleOnButtonUnhovered)
    onSelectedChangedBase=(&self:WidgetTree.Button_LeaveMatch.HandleOnSelectedChanged)
    style=/FortUILibrary/Foundation/Buttons/ButtonStyles/Skew/ButtonStyle-Skew_Green_M.ButtonStyle-Skew_Green_M_C
    triggeringInputAction {dataTable=&InputActionDataTable, rowName=GenericFaceButtonBottom}
   }
   slot {class=HorizontalBoxSlot}
  }
  Spacer_69 {
   widgetClassPath=&Spacer
   parent=HorizontalBox_47
   widget=Spacer_69
   properties {size=<64 1>}
   slot {class=HorizontalBoxSlot, properties {size {sizeRule=Fill, value=1}}}
  }
  Button_ContinueMatch {
   widgetClassPath=&IconTextButton
   bInherited=true
   bIsVariable=true
   parent=HorizontalBox_47
   widget=Button_ContinueMatch
   properties {
    buttonText=Continue Match
    onButtonBaseClicked=(&self:WidgetTree.Button_ContinueMatch.HandleOnButtonClicked)
    onButtonBaseDoubleClicked=(&self:WidgetTree.Button_ContinueMatch.HandleOnButtonDoubleClicked)
    onButtonBaseHovered=(&self:WidgetTree.Button_ContinueMatch.HandleOnButtonHovered)
    onButtonBaseUnhovered=(&self:WidgetTree.Button_ContinueMatch.HandleOnButtonUnhovered)
    onSelectedChangedBase=(&self:WidgetTree.Button_ContinueMatch.HandleOnSelectedChanged)
    style=/FortUILibrary/Foundation/Buttons/ButtonStyles/Skew/ButtonStyle-Skew_Red.ButtonStyle-Skew_Red_C
    triggeringInputAction {dataTable=&InputActionDataTable, rowName=GenericFaceButtonRight}
   }
   slot {class=HorizontalBoxSlot}
  }
 ]
}
